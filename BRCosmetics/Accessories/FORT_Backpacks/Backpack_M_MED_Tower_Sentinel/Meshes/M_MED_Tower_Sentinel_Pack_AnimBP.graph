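# Animation Blueprint for the Tower Sentinel backpack cosmetic, whose output pose is only the looked-up idle prop sequence.
# An unused chain of Transform (Modify) Bone nodes for procedural spine, left and right bone sway is not connected to the output.
# Its offsets come from Ambient Animation, a game-time sine helper called per bone in the EventGraph, where no event triggers the Sequence.
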AnimBlueprint /BRCosmetics/Accessories/FORT_Backpacks/Backpack_M_MED_Tower_Sentinel/Meshes/M_MED_Tower_Sentinel_Pack_AnimBP
target_skeleton=&/M_MED_Tower_Sentinel_Pack_Skeleton
graphs [
 Ambient Animation {nodes [
   n1 = Ambient Animation!() -> (
    then
    Time Offset 1:double=~
    Magnitude Multi:double=~
    Period:Vector=~
    Magnitude:Vector="10,0,0"
    Period multi:double=1
    Rotate Multi:double=~
   ) {is=title K2Node_FunctionEntry}
   n2 = BreakVector(InVec=@n1.Magnitude)
   n3 = BreakVector(InVec=@n1.Period)
   n4 = GetGameTimeInSeconds()
   n5 = float * float(A=float * float(A=@n1.Period multi, B=@n3.Y), B=@n4)
   n6 = float * float(A="Sin (Radians)"(A=float + float(A=@n5, B=@n1) {is=title K2Node_CommutativeAssociativeBinaryOperator}), B=@n2.Y)
   n7 = float * float(A=float * float(A=@n1.Period multi, B=@n3.Z), B=@n4)
   n8 = float * float(A="Sin (Radians)"(A=float + float(A=@n7, B=@n1)), B=@n2.Z)
   n9 = float * float(A=float * float(A=@n1.Period multi, B=@n3), B=@n4)
   n10 = float * float(A="Sin (Radians)"(A=float + float(A=@n9, B=@n1)), B=@n2)
   n11 = MakeRotator(Roll=@n10, Pitch=@n6, Yaw=@n8)
   > Return Node!(
    execute
    VectorOut:Vector=vector * float(A=MakeVector(X=@n10, Y=@n6, Z=@n8), B=@n1.Magnitude Multi)
    RotateOut:Rotator="Scale Rotator (Float)"!(A:Rotator="Scale Rotator (Float)"(A=@n11, B=@n1.Magnitude Multi), B:float=@n1.Rotate Multi) -> (ReturnValue:Rotator)
   )
  ]}
 AnimGraph {nodes [
   Local Space Ref Pose()
   n1 = "Transform (Modify) Bone - Bone: root"(ComponentPose=Local To Component(LocalPose=Input Pose()), Translation="0,0,0", Rotation="0,0,0") {
    rotationMode=BMM_Additive
    rotationSpace=BCS_BoneSpace
    translationMode=BMM_Additive
    translationSpace=BCS_BoneSpace
    alphaBoolBlend {blendOption=Linear}
    boneToModify {boneName=root}
   }
   n2 = "Transform (Modify) Bone - Bone: Dyn_Spine1"!(
    ComponentPose=@n1
    bAlphaBoolEnabled:bool=True
    Translation="0,0,0"
    Rotation=Get AmbientRot_Spine_01() -> (AmbientRot_Spine_01:Rotator)
    Scale="1,1,1"
   ) -> (Pose) {
    alphaInputType=Bool
    rotationMode=BMM_Replace
    rotationSpace=BCS_BoneSpace
    translationMode=BMM_Additive
    translationSpace=BCS_BoneSpace
    alphaBoolBlend {blendOption=Linear}
    boneToModify {boneName=Dyn_Spine1}
   }
   n3 = "Transform (Modify) Bone - Bone: Dyn_Spine2"!(
    ComponentPose=@n2
    bAlphaBoolEnabled:bool=True
    Translation="0,0,0"
    Rotation=Get AmbientRot_Spine_02() -> (AmbientRot_Spine_02:Rotator)
    Scale="1,1,1"
   ) -> (Pose) {
    alphaInputType=Bool
    rotationMode=BMM_Replace
    rotationSpace=BCS_BoneSpace
    translationMode=BMM_Additive
    translationSpace=BCS_BoneSpace
    alphaBoolBlend {blendOption=Linear}
    boneToModify {boneName=Dyn_Spine2}
   }
   n4 = "Transform (Modify) Bone - Bone: Dyn_Spine3"!(
    ComponentPose=@n3
    bAlphaBoolEnabled:bool=True
    Translation="0,0,0"
    Rotation=Get AmbientRot_Spine_03() -> (AmbientRot_Spine_03:Rotator)
    Scale="1,1,1"
   ) -> (Pose) {
    alphaInputType=Bool
    rotationMode=BMM_Replace
    rotationSpace=BCS_BoneSpace
    translationMode=BMM_Additive
    translationSpace=BCS_BoneSpace
    alphaBoolBlend {blendOption=Linear}
    boneToModify {boneName=Dyn_Spine3}
   }
   n5 = "Transform (Modify) Bone - Bone: Dyn_Spine4"!(
    ComponentPose=@n4
    bAlphaBoolEnabled:bool=True
    Translation="0,0,0"
    Rotation=Get AmbientRot_Spine_04() -> (AmbientRot_Spine_04:Rotator)
    Scale="1,1,1"
   ) -> (Pose) {
    alphaInputType=Bool
    rotationMode=BMM_Replace
    rotationSpace=BCS_BoneSpace
    translationMode=BMM_Additive
    translationSpace=BCS_BoneSpace
    alphaBoolBlend {blendOption=Linear}
    boneToModify {boneName=Dyn_Spine4}
   }
   n6 = "Transform (Modify) Bone - Bone: Dyn_Spine5"!(
    ComponentPose=@n5
    bAlphaBoolEnabled:bool=True
    Translation="0,0,0"
    Rotation=Get AmbientRot_Spine_05() -> (AmbientRot_Spine_05:Rotator)
    Scale="1,1,1"
   ) -> (Pose) {
    alphaInputType=Bool
    rotationMode=BMM_Replace
    rotationSpace=BCS_BoneSpace
    translationMode=BMM_Additive
    translationSpace=BCS_BoneSpace
    alphaBoolBlend {blendOption=Linear}
    boneToModify {boneName=Dyn_Spine5}
   }
   n7 = "Transform (Modify) Bone - Bone: Dyn_Spine6"!(
    ComponentPose=@n6
    bAlphaBoolEnabled:bool=True
    Translation=Get Ambient Trans Spine 06() -> (Ambient Trans Spine 06:Vector)
    Rotation=Get AmbientRot_Spine_06() -> (AmbientRot_Spine_06:Rotator)
    Scale="1,1,1"
   ) -> (Pose) {
    alphaInputType=Bool
    rotationMode=BMM_Replace
    rotationSpace=BCS_BoneSpace
    translationMode=BMM_Additive
    translationSpace=BCS_BoneSpace
    alphaBoolBlend {blendOption=Linear}
    boneToModify {boneName=Dyn_Spine6}
   }
   n8 = "Transform (Modify) Bone - Bone: Dyn_Spine7"!(
    ComponentPose=@n7
    bAlphaBoolEnabled:bool=True
    Translation=Get Ambient Trans Spine 07() -> (Ambient Trans Spine 07:Vector)
    Rotation=Get AmbientRot_Spine_07() -> (AmbientRot_Spine_07:Rotator)
    Scale="1,1,1"
   ) -> (Pose) {
    alphaInputType=Bool
    rotationMode=BMM_Replace
    rotationSpace=BCS_BoneSpace
    translationMode=BMM_Additive
    translationSpace=BCS_BoneSpace
    alphaBoolBlend {blendOption=Linear}
    boneToModify {boneName=Dyn_Spine7}
   }
   n9 = "Transform (Modify) Bone - Bone: Dyn_Spine8"!(
    ComponentPose=@n8
    bAlphaBoolEnabled:bool=True
    Translation="0,0,0"
    Rotation=Get AmbientRot_Spine_08() -> (AmbientRot_Spine_08:Rotator)
    Scale="1,1,1"
   ) -> (Pose) {
    alphaInputType=Bool
    rotationMode=BMM_Replace
    rotationSpace=BCS_BoneSpace
    translationMode=BMM_Additive
    translationSpace=BCS_BoneSpace
    alphaBoolBlend {blendOption=Linear}
    boneToModify {boneName=Dyn_Spine8}
   }
   n10 = "Transform (Modify) Bone - Bone: Dyn_Spine9"!(
    ComponentPose=@n9
    bAlphaBoolEnabled:bool=True
    Translation="0,0,0"
    Rotation=Get AmbientRot_Spine_09() -> (AmbientRot_Spine_09:Rotator)
    Scale="1,1,1"
   ) -> (Pose) {
    alphaInputType=Bool
    rotationMode=BMM_Replace
    rotationSpace=BCS_BoneSpace
    translationMode=BMM_Additive
    translationSpace=BCS_BoneSpace
    alphaBoolBlend {blendOption=Linear}
    boneToModify {boneName=Dyn_Spine9}
   }
   n11 = "Transform (Modify) Bone - Bone: Dyn_Spine10"!(
    ComponentPose=@n10
    bAlphaBoolEnabled:bool=True
    Translation="0,0,0"
    Rotation=Get AmbientRot_Spine_10() -> (AmbientRot_Spine_10:Rotator)
    Scale="1,1,1"
   ) -> (Pose) {
    alphaInputType=Bool
    rotationMode=BMM_Replace
    rotationSpace=BCS_BoneSpace
    translationMode=BMM_Additive
    translationSpace=BCS_BoneSpace
    alphaBoolBlend {blendOption=Linear}
    boneToModify {boneName=Dyn_Spine10}
   }
   n12 = "Transform (Modify) Bone - Bone: Dyn_Left1"!(
    ComponentPose=@n11
    bAlphaBoolEnabled:bool=True
    Translation="0,0,0"
    Rotation=Get AmbientRot_Left_01() -> (AmbientRot_Left_01:Rotator)
    Scale="1,1,1"
   ) -> (Pose) {
    alphaInputType=Bool
    rotationMode=BMM_Replace
    rotationSpace=BCS_BoneSpace
    translationMode=BMM_Additive
    translationSpace=BCS_BoneSpace
    alphaBoolBlend {blendOption=Linear}
    boneToModify {boneName=Dyn_Left1}
   }
   n13 = "Transform (Modify) Bone - Bone: Dyn_Left2"!(
    ComponentPose=@n12
    bAlphaBoolEnabled:bool=True
    Translation="0,0,0"
    Rotation=Get AmbientRot_Left_02() -> (AmbientRot_Left_02:Rotator)
    Scale="1,1,1"
   ) -> (Pose) {
    alphaInputType=Bool
    rotationMode=BMM_Replace
    rotationSpace=BCS_BoneSpace
    translationMode=BMM_Additive
    translationSpace=BCS_BoneSpace
    alphaBoolBlend {blendOption=Linear}
    boneToModify {boneName=Dyn_Left2}
   }
   n14 = "Transform (Modify) Bone - Bone: Dyn_Right3"!(
    ComponentPose=@n13
    bAlphaBoolEnabled:bool=True
    Translation="0,0,0"
    Rotation=Get AmbientRot_Right_03() -> (AmbientRot_Right_03:Rotator)
    Scale="1,1,1"
   ) -> (Pose) {
    alphaInputType=Bool
    rotationMode=BMM_Replace
    rotationSpace=BCS_BoneSpace
    translationMode=BMM_Additive
    translationSpace=BCS_BoneSpace
    alphaBoolBlend {blendOption=Linear}
    boneToModify {boneName=Dyn_Right3}
   }
   n15 = "Transform (Modify) Bone - Bone: Dyn_Right1"!(
    ComponentPose=@n14
    bAlphaBoolEnabled:bool=True
    Translation="0,0,0"
    Rotation=Get AmbientRot_Right_01() -> (AmbientRot_Right_01:Rotator)
    Scale="1,1,1"
   ) -> (Pose) {
    alphaInputType=Bool
    rotationMode=BMM_Replace
    rotationSpace=BCS_BoneSpace
    translationMode=BMM_Additive
    translationSpace=BCS_BoneSpace
    alphaBoolBlend {blendOption=Linear}
    boneToModify {boneName=Dyn_Right1}
   }
   n16 = "Transform (Modify) Bone - Bone: Dyn_Right2"!(
    ComponentPose=@n15
    bAlphaBoolEnabled:bool=True
    Translation="0,0,0"
    Rotation=Get AmbientRot_Right_02() -> (AmbientRot_Right_02:Rotator)
    Scale="1,1,1"
   ) -> (Pose) {
    alphaInputType=Bool
    rotationMode=BMM_Replace
    rotationSpace=BCS_BoneSpace
    translationMode=BMM_Additive
    translationSpace=BCS_BoneSpace
    alphaBoolBlend {blendOption=Linear}
    boneToModify {boneName=Dyn_Right2}
   }
   n17 = "Transform (Modify) Bone - Bone: Dyn_Right3"(ComponentPose=@n16, bAlphaBoolEnabled=True, Translation="0,0,0", Rotation=Get AmbientRot_Right_03(), Scale="1,1,1") {
    alphaInputType=Bool
    rotationMode=BMM_Replace
    rotationSpace=BCS_BoneSpace
    translationMode=BMM_Additive
    translationSpace=BCS_BoneSpace
    alphaBoolBlend {blendOption=Linear}
    boneToModify {boneName=Dyn_Right3}
   }
   n18 = "Transform (Modify) Bone - Bone: Dyn_Spine1"(
    ComponentPose=@n17
    bAlphaBoolEnabled=True
    Translation=Get Ambient Trans Spine 01() -> (Ambient Trans Spine 01:Vector)
    Rotation="0,0,0"
    Scale="1,1,1"
   ) {
    alphaInputType=Bool
    rotationMode=BMM_Replace
    rotationSpace=BCS_BoneSpace
    translationMode=BMM_Additive
    translationSpace=BCS_BoneSpace
    alphaBoolBlend {blendOption=Linear}
    boneToModify {boneName=Dyn_Spine1}
   }
   n19 = "Transform (Modify) Bone - Bone: Dyn_Spine1"(ComponentPose=@n18, bAlphaBoolEnabled=True, Translation=Get Ambient Trans Spine 07(), Rotation="0,0,0", Scale="1,1,1") {
    alphaInputType=Bool
    rotationMode=BMM_Replace
    rotationSpace=BCS_BoneSpace
    translationMode=BMM_Additive
    translationSpace=BCS_BoneSpace
    alphaBoolBlend {blendOption=Linear}
    boneToModify {boneName=Dyn_Spine1}
   }
   n20 = "Transform (Modify) Bone - Bone: Dyn_Left1"(
    ComponentPose=@n19
    bAlphaBoolEnabled=True
    Translation=Get Ambient Trans_Left_ 01() -> (Ambient Trans_Left_ 01:Vector)
    Rotation="0,0,0"
    Scale="1,1,1"
   ) {
    alphaInputType=Bool
    rotationMode=BMM_Replace
    rotationSpace=BCS_BoneSpace
    translationMode=BMM_Additive
    translationSpace=BCS_BoneSpace
    alphaBoolBlend {blendOption=Linear}
    boneToModify {boneName=Dyn_Left1}
   }
   n21 = "Transform (Modify) Bone - Bone: Dyn_Left2"(
    ComponentPose=@n20
    bAlphaBoolEnabled=True
    Translation=Get Ambient Trans_Left_ 02() -> (Ambient Trans_Left_ 02:Vector)
    Rotation="0,0,0"
    Scale="1,1,1"
   ) {
    alphaInputType=Bool
    rotationMode=BMM_Replace
    rotationSpace=BCS_BoneSpace
    translationMode=BMM_Additive
    translationSpace=BCS_BoneSpace
    alphaBoolBlend {blendOption=Linear}
    boneToModify {boneName=Dyn_Left2}
   }
   n22 = "Transform (Modify) Bone - Bone: Dyn_Left3"!(
    ComponentPose=@n21
    bAlphaBoolEnabled:bool=True
    Translation=Get Ambient Trans_Left_ 03() -> (Ambient Trans_Left_ 03:Vector)
    Rotation="0,0,0"
    Scale="1,1,1"
   ) -> (Pose) {
    alphaInputType=Bool
    rotationMode=BMM_Replace
    rotationSpace=BCS_BoneSpace
    translationMode=BMM_Additive
    translationSpace=BCS_BoneSpace
    alphaBoolBlend {blendOption=Linear}
    boneToModify {boneName=Dyn_Left3}
   }
   n23 = "Transform (Modify) Bone - Bone: Dyn_Right1"(
    ComponentPose=@n22
    bAlphaBoolEnabled=True
    Translation=Get Ambient Trans_Right_ 01() -> (Ambient Trans_Right_ 01:Vector)
    Rotation="0,0,0"
    Scale="1,1,1"
   ) {
    alphaInputType=Bool
    rotationMode=BMM_Replace
    rotationSpace=BCS_BoneSpace
    translationMode=BMM_Additive
    translationSpace=BCS_BoneSpace
    alphaBoolBlend {blendOption=Linear}
    boneToModify {boneName=Dyn_Right1}
   }
   n24 = "Transform (Modify) Bone - Bone: Dyn_Right2"(
    ComponentPose=@n23
    bAlphaBoolEnabled=True
    Translation=Get Ambient Trans_Right_ 02() -> (Ambient Trans_Right_ 02:Vector)
    Rotation="0,0,0"
    Scale="1,1,1"
   ) {
    alphaInputType=Bool
    rotationMode=BMM_Replace
    rotationSpace=BCS_BoneSpace
    translationMode=BMM_Additive
    translationSpace=BCS_BoneSpace
    alphaBoolBlend {blendOption=Linear}
    boneToModify {boneName=Dyn_Right2}
   }
   "Transform (Modify) Bone - Bone: Dyn_Right3"(
    ComponentPose=@n24
    bAlphaBoolEnabled=True
    Translation=Get Ambient Trans_Right_ 03() -> (Ambient Trans_Right_ 03:Vector)
    Rotation="0,0,0"
    Scale="1,1,1"
   ) {
    alphaInputType=Bool
    rotationMode=BMM_Replace
    rotationSpace=BCS_BoneSpace
    translationMode=BMM_Additive
    translationSpace=BCS_BoneSpace
    alphaBoolBlend {blendOption=Linear}
    boneToModify {boneName=Dyn_Right3}
   }
   n25 = Sequence Player 'Tower_Sentinel_Idle_Anim_PROP'() {sequence=/BRCosmetics/Accessories/FORT_Backpacks/Backpack_M_MED_Tower_Sentinel/Animation/Tower_Sentinel_Idle_Anim_PROP}
   Output Pose(Result=Component To Local(ComponentPose=Local To Component(LocalPose=@n25))) {layerGroup=None}
  ]}
 EventGraph {nodes [
   n1 = Sequence!(execute) -> (
    then_0
    then_1
    then_2:Exec
    then_3:Exec
    then_4:Exec
    then_5:Exec
    then_6:Exec
    then_7:Exec
    then_8:Exec
    then_9:Exec
    then_10:Exec
    then_11:Exec
    then_12:Exec
    then_13:Exec
    then_14:Exec
    then_15:Exec
    then_16:Exec
    then_17:Exec
    then_18:Exec
    then_19:Exec
    then_20:Exec
    then_21:Exec
    then_22:Exec
    then_23:Exec
   )
   > n2 = Ambient Animation!(
    execute
    self
    Time Offset 1:double=0
    Magnitude Multi:double=1
    Period:Vector="5,0,0"
    Magnitude:Vector="5,0,0"
    Period multi:double=1
    Rotate Multi:double=1
   ) -> (then, VectorOut:Vector, RotateOut:Rotator) {is=title K2Node_CallFunction}
   > Set Ambient Trans Spine 01(Ambient Trans Spine 01:Vector=@n2) -> (Output_Get:Vector)
   > Set AmbientRot_Spine_01(AmbientRot_Spine_01:Rotator=MakeRotator(Yaw=BreakRotator(InRot=@n2.RotateOut))) -> (Output_Get:Rotator)
   n3 = Ambient Animation(execute=@n1.then_1, Time Offset 1=0, Magnitude Multi=1, Period="3,5, 0", Magnitude="4,0,0", Period multi=1, Rotate Multi=1)
   > Set Ambient Trans Spine 02(Ambient Trans Spine 02:Vector=@n3) -> (Output_Get:Vector)
   > Set AmbientRot_Spine_02(AmbientRot_Spine_02:Rotator=MakeRotator(Yaw=BreakRotator(InRot=@n3.RotateOut))) -> (Output_Get:Rotator)
   n4 = Ambient Animation(execute=@n1.then_10, Time Offset 1=0, Magnitude Multi=1, Period="1,0,1", Magnitude="1,0,1", Period multi=1, Rotate Multi=1)
   > Set Ambient Trans_Left_ 01(Ambient Trans_Left_ 01:Vector=@n4) -> (Output_Get:Vector)
   n5 = Ambient Animation(execute=@n1.then_11, Time Offset 1=0, Magnitude Multi=1, Period="1,0,0.5", Magnitude="2,0,0.5", Period multi=1, Rotate Multi=1)
   > Set Ambient Trans_Left_ 02(Ambient Trans_Left_ 02:Vector=@n5) -> (Output_Get:Vector)
   n6 = Ambient Animation(execute=@n1.then_12, Time Offset 1=0, Magnitude Multi=1, Period="1,0,0.5", Magnitude="2,0,0.5", Period multi=1, Rotate Multi=1)
   > Set Ambient Trans_Left_ 03(Ambient Trans_Left_ 03:Vector=@n6) -> (Output_Get:Vector)
   n7 = Ambient Animation(execute=@n1.then_13, Time Offset 1=0, Magnitude Multi=1, Period="4, 0, 0", Magnitude="5,0,0", Period multi=1, Rotate Multi=1)
   > Set AmbientRot_Right_03(AmbientRot_Right_03:Rotator=MakeRotator(Yaw=BreakRotator(InRot=@n7.RotateOut))) -> (Output_Get:Rotator)
   n8 = Ambient Animation(execute=@n1.then_14, Time Offset 1=0, Magnitude Multi=1, Period="1.5,0,0.5", Magnitude="1,0,0.5", Period multi=1, Rotate Multi=1)
   > Set Ambient Trans_Right_ 02(Ambient Trans_Right_ 02:Vector=@n8) -> (Output_Get:Vector)
   n9 = Ambient Animation(execute=@n1.then_15, Time Offset 1=0, Magnitude Multi=1, Period="1,0,1", Magnitude="1,0,1", Period multi=1, Rotate Multi=1)
   > Set Ambient Trans_Right_ 01(Ambient Trans_Right_ 01:Vector=@n9) -> (Output_Get:Vector)
   n10 = Ambient Animation(execute=@n1.then_16, Time Offset 1=0, Magnitude Multi=1, Period="1,0,1", Magnitude="2,0,1", Period multi=1, Rotate Multi=1)
   > Set Ambient Trans_Right_ 03(Ambient Trans_Right_ 03:Vector=@n10) -> (Output_Get:Vector)
   n11 = Ambient Animation(execute=@n1.then_17, Time Offset 1=0, Magnitude Multi=1, Period="3, 0, 0", Magnitude="10,0,0", Period multi=1, Rotate Multi=1)
   > Set AmbientRot_Left_01(AmbientRot_Left_01:Rotator=MakeRotator(Yaw=BreakRotator(InRot=@n11.RotateOut))) -> (Output_Get:Rotator)
   n12 = Ambient Animation(execute=@n1.then_18, Time Offset 1=0, Magnitude Multi=1, Period="2, 0, 0", Magnitude="8,0,0", Period multi=1, Rotate Multi=1)
   > Set AmbientRot_Left_02(AmbientRot_Left_02:Rotator=MakeRotator(Yaw=BreakRotator(InRot=@n12.RotateOut))) -> (Output_Get:Rotator)
   n13 = Ambient Animation(execute=@n1.then_19, Time Offset 1=0, Magnitude Multi=1, Period="2,0,0", Magnitude="10,0,0", Period multi=1, Rotate Multi=1)
   > Set AmbientRot_Left_03(AmbientRot_Left_03:Rotator=MakeRotator(Yaw=BreakRotator(InRot=@n13.RotateOut))) -> (Output_Get:Rotator)
   n14 = Ambient Animation(execute=@n1.then_2, Time Offset 1=0, Magnitude Multi=1, Period="2,5, 0", Magnitude="3,0,0", Period multi=1, Rotate Multi=1)
   > Set Ambient Trans Spine 03(Ambient Trans Spine 03:Vector=@n14) -> (Output_Get:Vector)
   > Set AmbientRot_Spine_03(AmbientRot_Spine_03:Rotator=MakeRotator(Yaw=BreakRotator(InRot=@n14.RotateOut))) -> (Output_Get:Rotator)
   n15 = Ambient Animation(execute=@n1.then_20, Time Offset 1=0, Magnitude Multi=1, Period="5, 0, 0", Magnitude="5,0,0", Period multi=1, Rotate Multi=1)
   > Set AmbientRot_Right_01(AmbientRot_Right_01:Rotator=MakeRotator(Yaw=BreakRotator(InRot=@n15.RotateOut))) -> (Output_Get:Rotator)
   n16 = Ambient Animation(execute=@n1.then_21, Time Offset 1=0, Magnitude Multi=1, Period="3,0,0", Magnitude="4,0,0", Period multi=1, Rotate Multi=1)
   > Set AmbientRot_Right_02(AmbientRot_Right_02:Rotator=MakeRotator(Yaw=BreakRotator(InRot=@n16.RotateOut))) -> (Output_Get:Rotator)
   n17 = Ambient Animation(execute=@n1.then_22, Time Offset 1=0, Magnitude Multi=1, Period="0,1,0", Magnitude="0,1,0", Period multi=1, Rotate Multi=1)
   > Set Ambient Trans Spine 01(Ambient Trans Spine 01=@n17)
   n18 = Ambient Animation(execute=@n1.then_23, Time Offset 1=0, Magnitude Multi=1, Period="0,1,0", Magnitude="0,2,0", Period multi=1, Rotate Multi=1)
   > Set Ambient Trans Spine 07(Ambient Trans Spine 07:Vector=@n18) -> (Output_Get:Vector)
   n19 = Ambient Animation(execute=@n1.then_3, Time Offset 1=0, Magnitude Multi=1, Period="4, 0, 0", Magnitude="2,0,0", Period multi=1, Rotate Multi=1)
   > Set Ambient Trans Spine 04(Ambient Trans Spine 04:Vector=@n19) -> (Output_Get:Vector)
   > Set AmbientRot_Spine_04(AmbientRot_Spine_04:Rotator=MakeRotator(Yaw=BreakRotator(InRot=@n19.RotateOut))) -> (Output_Get:Rotator)
   n20 = Ambient Animation(execute=@n1.then_4, Time Offset 1=0, Magnitude Multi=1, Period="2, 0, 0", Magnitude="8,0,0", Period multi=1, Rotate Multi=1)
   > Set Ambient Trans Spine 05(Ambient Trans Spine 05:Vector=@n20) -> (Output_Get:Vector)
   > Set AmbientRot_Spine_05(AmbientRot_Spine_05:Rotator=MakeRotator(Yaw=BreakRotator(InRot=@n20.RotateOut))) -> (Output_Get:Rotator)
   n21 = Ambient Animation(execute=@n1.then_5, Time Offset 1=0, Magnitude Multi=1, Period="2, 0,0.5", Magnitude="1,0,0.5", Period multi=1, Rotate Multi=1)
   > Set Ambient Trans Spine 06(Ambient Trans Spine 06:Vector=@n21) -> (Output_Get:Vector)
   > Set AmbientRot_Spine_06(AmbientRot_Spine_06:Rotator=MakeRotator(Yaw=BreakRotator(InRot=@n21.RotateOut))) -> (Output_Get:Rotator)
   n22 = Ambient Animation(execute=@n1.then_6, Time Offset 1=0, Magnitude Multi=1, Period="5, 0,0.5", Magnitude="0.5,0,0.5", Period multi=1, Rotate Multi=1)
   > Set Ambient Trans Spine 07(Ambient Trans Spine 07=@n22)
   > Set AmbientRot_Spine_07(AmbientRot_Spine_07:Rotator=MakeRotator(Yaw=BreakRotator(InRot=@n22.RotateOut))) -> (Output_Get:Rotator)
   n23 = Ambient Animation(execute=@n1.then_7, Time Offset 1=0, Magnitude Multi=1, Period="2, 0, 0", Magnitude="10,0,0", Period multi=1, Rotate Multi=1)
   > Set Ambient Trans Spine 08(Ambient Trans Spine 08:Vector=@n23) -> (Output_Get:Vector)
   > Set AmbientRot_Spine_08(AmbientRot_Spine_08:Rotator=MakeRotator(Yaw=BreakRotator(InRot=@n23.RotateOut))) -> (Output_Get:Rotator)
   n24 = Ambient Animation(execute=@n1.then_8, Time Offset 1=0, Magnitude Multi=1, Period="2, 0, 0", Magnitude="3,0,0", Period multi=1, Rotate Multi=1)
   > Set Ambient Trans Spine 09(Ambient Trans Spine 09:Vector=@n24) -> (Output_Get:Vector)
   > Set AmbientRot_Spine_09(AmbientRot_Spine_09:Rotator=MakeRotator(Yaw=BreakRotator(InRot=@n24.RotateOut))) -> (Output_Get:Rotator)
   n25 = Ambient Animation(execute=@n1.then_9, Time Offset 1=0, Magnitude Multi=1, Period="2, 0, 0", Magnitude="10,0,0", Period multi=1, Rotate Multi=1)
   > Set Ambient Trans Spine 10(Ambient Trans Spine 10:Vector=@n25) -> (Output_Get:Vector)
   > Set AmbientRot_Spine_10(AmbientRot_Spine_10:Rotator=MakeRotator(Yaw=BreakRotator(InRot=@n25.RotateOut))) -> (Output_Get:Rotator)
   TryGetPawnOwner()
  ]}
]
variables [
 Ambient Trans Spine 01 {type=Vector}
 Ambient Trans Spine 02 {type=Vector}
 Ambient Trans Spine 03 {type=Vector}
 Ambient Trans Spine 04 {type=Vector}
 Ambient Trans Spine 05 {type=Vector}
 Ambient Trans Spine 06 {type=Vector}
 Ambient Trans Spine 07 {type=Vector}
 Ambient Trans Spine 08 {type=Vector}
 Ambient Trans Spine 09 {type=Vector}
 Ambient Trans Spine 10 {type=Vector}
 Ambient Trans_Left_ 01 {type=Vector}
 Ambient Trans_Left_ 02 {type=Vector}
 Ambient Trans_Left_ 03 {type=Vector}
 Ambient Trans_Right_ 01 {type=Vector}
 Ambient Trans_Right_ 02 {type=Vector}
 Ambient Trans_Right_ 03 {type=Vector}
 AmbientRot_Left_01 {type=Rotator}
 AmbientRot_Left_02 {type=Rotator}
 AmbientRot_Left_03 {type=Rotator}
 AmbientRot_Right_01 {type=Rotator}
 AmbientRot_Right_02 {type=Rotator}
 AmbientRot_Right_03 {type=Rotator}
 AmbientRot_Spine_01 {type=Rotator}
 AmbientRot_Spine_02 {type=Rotator}
 AmbientRot_Spine_03 {type=Rotator}
 AmbientRot_Spine_04 {type=Rotator}
 AmbientRot_Spine_05 {type=Rotator}
 AmbientRot_Spine_06 {type=Rotator}
 AmbientRot_Spine_07 {type=Rotator}
 AmbientRot_Spine_08 {type=Rotator}
 AmbientRot_Spine_09 {type=Rotator}
 AmbientRot_Spine_10 {type=Rotator}
 Time Offset 1 {type=number}
]
